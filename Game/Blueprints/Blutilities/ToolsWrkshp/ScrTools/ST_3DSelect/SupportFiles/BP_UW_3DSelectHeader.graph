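# Header widget for the 3DSelect editor screen tool, laid out as a dark bordered column with two expandable boxes, Readout and Shortcuts.
# Variables hold readout line widget references and a map of shortcut labels for visibility cycling, volume resizing and scaling, and collision display.
# An EventForceSnapBoundsClicked dispatcher is declared.
# The event graph has only unconnected PreConstruct, Construct and Tick events, so the widget itself contains no logic to fill the boxes.
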WidgetBlueprint /Game/Blueprints/Blutilities/ToolsWrkshp/ScrTools/ST_3DSelect/SupportFiles/BP_UW_3DSelectHeader
graphs [
 EventForceSnapBoundsClicked {nodes [EventForceSnapBoundsClicked() {is=title K2Node_FunctionEntry}]}
 EventGraph {nodes [
   Event PreConstruct()
   Event Construct()
   Event Tick()
  ]}
]
variables [
 EventForceSnapBoundsClicked {}
 Readout Line Actor {type=BP_UW_ReadoutLine_C, default=null}
 Readout Line Class {type=BP_UW_ReadoutLine_C, default=null}
 ReadoutLine_Selected {type=BP_UW_ReadoutLine_C, default=null}
 ShortcutIndex {type=integer}
 ShortcutLabels {
  type="map<string>"
  default {Cycle Object Visibility=Press V, Resize Volume=Hold X, Scale Volume Uniformly=Hold X + Mouse Wheel, Show Collision Geo=Press C}
 }
 ShortcutLines {type="array<BP_UW_ReadoutLine_C>", default []}
]
widgets {
 info {inheritedWidgetCount=0, parentClass=&UserWidget, rootWidgetClass=&Border, widgetCount=4}
 widgets [
  Border_25 {widgetClassPath=&Border, widget=Border_25, properties {brushColor=<rgba 0.0176 0.0176 0.0176 1>, padding=<ltrb 0 2 0 0>}}
  VerticalBox_25 {
   widgetClassPath=&VerticalBox
   parent=Border_25
   widget=VerticalBox_25
   slot {class=BorderSlot, properties {padding=<ltrb 0 2 0 0>}}
  }
  Expandable_Readout {
   widgetClassPath=/Game/Blueprints/Blutilities/ToolsWrkshp/ScrTools/Shared/SharedWidgets/ExpandableBox/BP_UW_ExpandableBox
   bIsVariable=true
   parent=VerticalBox_25
   widget=Expandable_Readout
   properties {boxColor=<rgba 0 0 0 0>}
   slot {class=VerticalBoxSlot}
  }
  Expandable_Shortcuts {
   widgetClassPath=&BP_UW_ExpandableBox
   bIsVariable=true
   parent=VerticalBox_25
   widget=Expandable_Shortcuts
   properties {boxColor=<rgba 0 0 0 0>}
   slot {class=VerticalBoxSlot}
  }
 ]
}
